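; GA_S17_Quest_Altitude is a Season 17 battle pass quest ability that detects when its owning player pawn reaches a set altitude.
; On activation it calls the parent and starts a looping timer that runs ElevationCheck every two seconds.
; ElevationCheck reads the pawn's world Z and, at or above a fixed threshold, applies the GE_Quest_Altitude gameplay effect to that pawn.
; The check repeats, and nothing in this graph clears the timer.
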(event Custom|ElevationCheck
  (bind _owningplayerpawn (Variables|GAVars|Data|GetOwningPlayerPawn))
  (bind (_returnvalue_x _returnvalue_y _returnvalue_z) (Math|Vector|BreakVector (Transformation|GetActorLocation _owningplayerpawn)))
  (if (>= _returnvalue_z 5e+03)
    (Fortnite|ApplyGameplayEffectsToActor _owningplayerpawn (Utilities|Array|MakeArray "/BattlepassS17/Items/QuestAbilities/Quest_Altitude/GE_Quest_Altitude.GE_Quest_Altitude_C") (0.0 0.0 0.0) false 0 0 "()")))

(event Ability|EventActivateAbility
  (|CallParentFunction)
  (Utilities|Time|SetTimerbyEvent (delegate Custom|ElevationCheck) 2.0 true))

(event Ability|EventFailedToActivatePassiveAbility
  (params (ActorInfo :type Gameplay%20Ability%20Actor%20Info%20Structure)))

(event Ability|EventOnEndAbility
  (params (bWasCancelled :type Boolean)))
